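# Stereo MetaSound source that plays the looping Secret Lair music for the Season 41 DuckYacht POI music encounter.
# It wraps the shared MSS_MusicEncounter_VolumeLoop_NoClock MetaSound, forwarding the wave, volume, fade, pause and falloff inputs and passing its stereo audio and finished trigger out.
# The source's attenuation distance input is not wired into that node.
# It uses the POI music sound class and overrides concurrency to one instance.
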MetaSoundSource /Hera/Audio/POI/DuckYacht/MSS_S41_MusicEncounter_SecretLair_Music
inputs [
 FadeInTime {type=Time}
 MasterVolume {type=Float}
 Music.WaveAsset {type=WaveAsset}
 MusicEncounter.Attenuation.FalloffDistance {type=Float}
 MusicEncounter.Attenuation.InnerRadius {type=Float}
 MusicEncounter.bIsPaused {type=Bool}
 MusicEncounter.PauseFadeOutTime {type=Time}
 UE.Attenuation.Distance {type=Float}
 UE.Source.OnPlay {type=Trigger}
]
interfaces [UE.Attenuation:1.0, UE.OutputFormat.Stereo:1.0, UE.Source.OneShot:1.0, UE.Source:1.1]
outputs [
 "UE.OutputFormat.Stereo.Audio:0" {type=Audio}
 "UE.OutputFormat.Stereo.Audio:1" {type=Audio}
 UE.Source.OneShot.OnFinished {type=Trigger}
]
pages [Default {
  input_defaults {
   FadeInTime=2
   MasterVolume=0.2
   Music.WaveAsset=&/MX_S41_SecretLair_LP
   MusicEncounter.Attenuation.FalloffDistance=3000
   MusicEncounter.Attenuation.InnerRadius=3000
   MusicEncounter.PauseFadeOutTime=2
   MusicEncounter.bIsPaused=null
   UE.Attenuation.Distance=0
   UE.Source.OnPlay=null
  }
  nodes [MSS_MusicEncounter_VolumeLoop_NoClock_0 = /MusicEncounter/MetaSounds/MSS_MusicEncounter_VolumeLoop_NoClock(
    FadeInTime=^FadeInTime
    MasterVolume=^MasterVolume
    Music.WaveAsset=^Music.WaveAsset
    MusicEncounter.Attenuation.FalloffDistance=^MusicEncounter.Attenuation.FalloffDistance
    MusicEncounter.Attenuation.InnerRadius=^MusicEncounter.Attenuation.InnerRadius
    MusicEncounter.PauseFadeOutTime=^MusicEncounter.PauseFadeOutTime
    MusicEncounter.bIsPaused=^MusicEncounter.bIsPaused
    UE.Source.OnPlay=^UE.Source.OnPlay
   ) {also {defaults {UE.Attenuation.Distance=1}}}]
  links [
   {dst="@UE.OutputFormat.Stereo.Audio:0", src=MSS_MusicEncounter_VolumeLoop_NoClock_0.UE.OutputFormat.Stereo.Audio:0}
   {dst="@UE.OutputFormat.Stereo.Audio:1", src=MSS_MusicEncounter_VolumeLoop_NoClock_0.UE.OutputFormat.Stereo.Audio:1}
   {dst="@UE.Source.OneShot.OnFinished", src=MSS_MusicEncounter_VolumeLoop_NoClock_0.UE.Source.OneShot.OnFinished}
  ]
 }]
properties {
 assetImportData=&self:AssetImportData
 bOverrideConcurrency=true
 numChannels=2
 outputFormat=Stereo
 sampleRate=48000
 soundClassObject=/Game/Sounds/FortSoundClassesAndModes/Fort_Music_POI_PSM
 concurrencyOverrides {maxCount=1}
 concurrencySet [null]
}
