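# Stereo MetaSound source for the WaddleJam dance floor music at the DuckYacht POI, spatialized by its own attenuation settings and routed to the POI music sound class.
# On play, a stinger wave player starts, and its finish triggers a second player holding the loop track.
# Both players feed a two-input stereo mixer to the output.
# The DuckParty trigger input is declared but connects to nothing.
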MetaSoundSource /Hera/Audio/POI/DuckYacht/MSS_S41_WaddleJam_DanceFloor
inputs [
 DuckParty {type=Trigger}
 UE.Source.OnPlay {type=Trigger}
]
interfaces [UE.OutputFormat.Stereo:1.0, UE.Source:1.1]
outputs [
 "UE.OutputFormat.Stereo.Audio:0" {type=Audio}
 "UE.OutputFormat.Stereo.Audio:1" {type=Audio}
]
pages [Default {
  input_defaults {DuckParty=null, UE.Source.OnPlay=null}
  nodes [
   WavePlayer_0 = UE.Wave Player:Stereo(Play=^UE.Source.OnPlay) {also {defaults {Wave Asset=&/MX_S41_WaddleJam_DanceFloor_STNG}, version [1,1]}}
   WavePlayer_1 = UE.Wave Player:Stereo(Play=@WavePlayer_0.On Finished) {also {defaults {Loop="<MetaSoundBuilderResult.SUCCEEDED: 0>", Wave Asset=&/MX_S41_WaddleJam_DanceFloor_LP}, version [1,1]}}
   AudioMixerAudioMixerStereo2_0 = "AudioMixer.Audio Mixer (Stereo, 2)"(In 0 L=@WavePlayer_0.Out Left, In 0 R=@WavePlayer_0.Out Right, In 1 L=@WavePlayer_1.Out Left, In 1 R=@WavePlayer_1.Out Right)
  ]
  links [
   {dst="@UE.OutputFormat.Stereo.Audio:0", src=AudioMixerAudioMixerStereo2_0.Out L}
   {dst="@UE.OutputFormat.Stereo.Audio:1", src=AudioMixerAudioMixerStereo2_0.Out R}
  ]
 }]
properties {
 assetImportData=&self:AssetImportData
 attenuationSettings=&/ATT_WaddleJam_DanceFloor
 bLooping=true
 duration=10000
 maxDistance=4500
 numChannels=2
 outputFormat=Stereo
 sampleRate=48000
 soundClassObject=/Game/Sounds/FortSoundClassesAndModes/Fort_Music_POI_PSM
 volume=0.8
}
