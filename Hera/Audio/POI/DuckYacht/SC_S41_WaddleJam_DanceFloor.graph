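# Sound cue for the WaddleJam dance floor music stinger at the DuckYacht POI.
# A single wave player feeds an attenuation node that reuses the Athena bus lobby engine close-range settings.
# An Enveloper at the root eases the volume down to half over the first five seconds.
# The cue routes to the POI music sound class.
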SoundCue /Hera/Audio/POI/DuckYacht/SC_S41_WaddleJam_DanceFloor
root=Enveloper_0
nodes [
 {id=Attenuation_5, children [WavePlayer_3], properties {attenuationSettings=/Game/Sounds/Attenuation/Fort_Bus_Lobby_Engine_Athena_Close_Att}}
 {
  id=Enveloper_0
  children [Attenuation_5]
  properties {pitchCurve {editorCurveData {keys [{time=0}]}}, volumeCurve {editorCurveData {keys [<at 0 1>,<at 5 0.5>]}}}
 }
 {id=WavePlayer_3, properties {soundWaveAssetPtr=&/MX_S41_WaddleJam_DanceFloor_STNG}}
]
properties {maxDistance=12500, soundClassObject=/Game/Sounds/FortSoundClassesAndModes/Fort_Music_POI_PSM, volumeMultiplier=0.8}
